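# Masked, two-sided lit material for Chinese lantern mesh particles in a skydiving contrail effect.
# Diffuse, normal, MSR and opacity mask textures are parameters.
# Emissive pulses on a sine, with speed and phase randomized by a dynamic parameter, then passes through time-of-day (dimmer by day) and FX emissive functions.
# A camera-proximity sphere mask builds an offset that collapses vertices toward the particle center, zeroed on ES3_1.
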Material /BRCosmetics/Athena/Effects/Materials/Skydiving/P_Contrails_ChineseLantern/M_LanternMesh
graph {
 properties {
  bUsedWithMeshParticles=true
  bUsedWithNiagaraMeshParticles=true
  blendMode=BLEND_Masked
  pixelDepthOffsetMode=PDOM_Legacy
  twoSided=true
  usedShadingModels=Default Lit
 }
 nodes [
  Constant() {r=0.5}
  n1 = ParticlePositionWS()
  n2 = CameraPositionWS()
  n3 = Constant() {r=100}
  n4 = Constant() {r=0.1}
  n5 = Multiply(A=Subtract(A=@n1, B=WorldPosition().XYZ), B=Saturate(None=SphereMask(A=@n1, B=@n2, Radius=@n3, Hardness=@n4))) {constB=0.5}
  FeatureLevelSwitch(Default=@n5, ES3_1=Constant(), SM4_REMOVED=@n5)
  n6 = DynamicParameter() {paramNames [Param1,Param2,Param3,Random]}
  n7 = Frac(None=@n6.Random)
  n8 = Add(A=Multiply(A=Time(), B=LinearInterpolate(Alpha=@n7) {constA=1, constB=2}), B=@n6.Random)
  n9 = Multiply(
   A=Desaturation(
    None=TextureSample() {texture=&/T_Lantern_E}.RGB
    Fraction=Constant() {r=0.1}
   )
   B=LinearInterpolate(Alpha=Sine(None=@n8)) {constA=3.5, constB=6}
  )
  MaterialFunctionCall(Emissive=MaterialFunctionCall(Color_Morning=@n9, Color_Day=Multiply(A=@n9) {constB=0.15}, Color_Evening=@n9, Color_Night=@n9) {materialFunction=&Color_Time_of_Day}.Result) {materialFunction=&MF_FX_EmissiveAdjustments}
  ObjectPositionWS()
  TextureSampleParameter2D() {parameterName=Diffuse Texture, texture=&/T_Lanterns_D}
  TextureSampleParameter2D() {
   parameterName=Normal Texture
   samplerType=SAMPLERTYPE_Normal
   texture=&/T_Lanterns_N
  }
  TextureSampleParameter2D() {
   parameterName=Opacity Mask
   samplerType=SAMPLERTYPE_LinearGrayscale
   texture=&/T_Lanterns_M
  }
  TextureSampleParameter2D() {
   parameterName=MSR Texture
   samplerType=SAMPLERTYPE_Masks
   texture=&/T_Lanterns_S
  }
 ]
}
