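# Translucent, two-sided silk material for the Chinese lantern skydiving contrail, built for beam trails and Niagara ribbons.
# Red emissive follows time of day, dims by face side, and varies sinusoidally across V.
# Sine and cosine of trail UV plus object height drive optional WPO sway and edge flutter, scaled along the trail, and also build a normal.
# A time offset replaces falling motion in frontend preview; a camera sphere mask supplies a proximity fade.
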Material /BRCosmetics/Athena/Effects/Materials/Skydiving/P_Contrails_ChineseLantern/M_LanternSilk
graph {
 properties {
  bUsedWithBeamTrails=true
  bUsedWithNiagaraRibbons=true
  blendMode=BLEND_Translucent
  pixelDepthOffsetMode=PDOM_Legacy
  twoSided=true
  usedShadingModels=Default Lit
 }
 nodes [
  n1 = MaterialFunctionCall!(MaxiumURange, RangeRampHardness, MirrorTexCoordU) {materialFunction=/BRCosmetics/Weapons/FORT_Melee/Pickaxe_Venus/MF_LocalSpaceTrailUV}
  n2 = OneMinus(None=@n1."LocalSpaceCoordinate(UV)")
  Clamp(None=@n2)
  Constant() {r=0.8}
  n3 = Sine(None=ComponentMask(None=TextureCoordinate() {vTiling=0.5}) {g=true})
  n4 = WorldPosition()
  n5 = CameraPositionWS()
  n6 = Constant() {r=100}
  n7 = Constant()
  LinearInterpolate(B=@n3, Alpha=Multiply(A=SphereMask(A=@n4.XYZ, B=@n5, Radius=@n6, Hardness=@n7)) {constB=1.5}) {constA=1}
  n8 = MaterialFunctionCall() {materialFunction=&Color_Time_of_Day}
  n9 = If(A=TwoSidedSign(), "A > B"=@n8.Result, "A < B"=Multiply(A=@n8.Result) {constB=0.2}) {constB=0.5}
  MaterialFunctionCall(Emissive=Multiply(A=Constant3Vector() {constant=<rgba 1 0.05 0.05 0>}, B=Multiply(A=@n9, B=@n3))) {materialFunction=&MF_FX_EmissiveAdjustments}
  Multiply(A=@n2) {constB=5}
  n10 = ObjectPositionWS()
  n11 = CollectionParameter() {collection=&FortniteMaterialParameters, parameterName=AthenaFrontendPreview}
  n12 = Multiply(A=Multiply(A=@n11, B=Time())) {constB=-2500}
  n13 = ComponentMask(None=Multiply(A=Add(A=@n10, B=@n12)) {constB=-1}) {b=true}
  n14 = ScalarParameter() {defaultValue=2500, parameterName=OffsetSpeed}
  n15 = Add(
   A=ComponentMask(None=@n1."LocalSpaceCoordinate(UV)") {r=true}
   B=Add(A=Divide(A=@n13, B=@n14) {constB=3500}, B=ScalarParameter() {parameterName=RandomOffset})
  )
  n16 = Multiply(A=AppendVector(A=Sine(None=@n15), B=Cosine(None=@n15))) {constB=75}
  n17 = ComponentMask(None=Clamp(None=Subtract(A=@n2) {constB=0.2})) {r=true}
  n18 = ComponentMask(None=TextureCoordinate() {vTiling=2}) {g=true}
  n19 = Multiply(A=@n15) {constB=3}
  n20 = AppendVector(A=Cosine(None=@n19), B=Sine(None=@n19))
  n21 = Multiply(A=Multiply(A=Subtract(A=@n18), B=@n20) {constB=8}) {constB=10}
  n22 = ComponentMask(None=@n2) {r=true}
  StaticSwitchParameter(True=Add(A=Multiply(A=@n16, B=@n17), B=Multiply(A=@n21, B=@n22)), False=Constant()) {defaultValue=true, parameterName=UseWPO}
  n23 = ComponentMask(None=VertexNormalWS()) {b=true}
  Transform(None=Normalize(VectorInput=AppendVector(A=@n23, B=@n20))) {transformSourceType=TRANSFORMSOURCE_World, transformType=TRANSFORM_Local}
 ]
}
