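# Translucent two-sided lit material for silk ring mesh particles in the Chinese lantern skydiving contrail.
# Emissive tints BaseColor by time of day, darkens backfaces and shades along V with a sine.
# Opacity combines that gradient, a camera-proximity sphere mask and a cutoff along U driven by particle parameter Wride.
# World position offset ripples the silk using object position, substituting time-driven motion in frontend preview.
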Material /BRCosmetics/Athena/Effects/Materials/Skydiving/P_Contrails_ChineseLantern/M_LanternSilkRing
graph {
 properties {
  bUsedWithMeshParticles=true
  bUsedWithNiagaraMeshParticles=true
  blendMode=BLEND_Translucent
  pixelDepthOffsetMode=PDOM_Legacy
  twoSided=true
  usedShadingModels=Default Lit
 }
 nodes [
  n1 = MaterialFunctionCall() {materialFunction=&Color_Time_of_Day}
  n2 = If(A=TwoSidedSign(), "A > B"=@n1.Result, "A < B"=Multiply(A=@n1.Result) {constB=0.2}) {constB=0.5}
  n3 = Sine(None=ComponentMask(None=TextureCoordinate() {vTiling=0.5}) {g=true})
  MaterialFunctionCall(Emissive=Multiply(A=VectorParameter() {parameterName=BaseColor, defaultValue=<rgba 1 0.05 0.05 1>}.RGB, B=Multiply(A=@n2, B=@n3))) {materialFunction=&MF_FX_EmissiveAdjustments}
  n4 = SphereMask(A=WorldPosition().XYZ, B=CameraPositionWS(), Radius=Constant() {r=100}, Hardness=Constant())
  n5 = TextureCoordinate() {vTiling=2}
  n6 = ComponentMask(None=@n5) {r=true}
  n7 = If(
   A=OneMinus(None=@n6)
   B=DynamicParameter() {defaultValue=<rgba 1 1 0 1>, paramNames [Param1, Param2, FrontendPreview, Wride]}.Wride
   "A > B"=Constant()
   "A < B"=Constant() {r=10}
  )
  Multiply(A=LinearInterpolate(B=@n3, Alpha=Multiply(A=@n4) {constB=1.5}) {constA=1}, B=Saturate(None=@n7))
  n8 = Transform(None=Constant3Vector() {constant=<rgba 0 1 0 0>})
  n9 = ObjectPositionWS()
  n10 = CollectionParameter() {collection=&FortniteMaterialParameters, parameterName=AthenaFrontendPreview}
  n11 = Add(A=DynamicParameter() {defaultValue=<rgba 1 1 0 1>, paramNames [Param1, Param2, FrontendPreview, Wride]}.FrontendPreview, B=@n10)
  n12 = Multiply(A=Multiply(A=Clamp(None=@n11), B=Time())) {constB=-2500}
  n13 = Add(
   A=Divide(A=Add(A=@n9, B=@n12), B=ScalarParameter() {defaultValue=1000, parameterName=OffsetSpeed}) {constB=1000}
   B=ScalarParameter() {parameterName=RandomOffset}
  )
  n14 = ComponentMask(None=@n13) {b=true}
  n15 = Multiply(A=@n6, B=@n6)
  n16 = Add(A=Multiply(A=@n14) {constB=-1}, B=Multiply(A=@n15) {constB=1.5})
  n17 = Sine(None=@n16)
  n18 = Transform(None=Constant3Vector() {constant=<rgba 0 0 1 0>})
  n19 = Cosine(None=@n16)
  n20 = Multiply(A=Add(A=Multiply(A=@n8, B=@n17), B=Multiply(A=@n18, B=@n19)), B=Multiply(A=OneMinus(None=@n15), B=@n7) {constA=10}) {constB=5}
  n21 = Subtract(A=ComponentMask(None=@n5) {g=true})
  n22 = OneMinus(None=ComponentMask(None=TextureCoordinate()) {r=true})
  n23 = Add(A=@n14, B=Multiply(A=@n22) {constB=2})
  n24 = Add(A=ObjectOrientation(), B=Multiply(A=@n13) {constB=0.5})
  n25 = Multiply(A=Multiply(A=Sine(None=@n23), B=Sine(None=@n24)), B=Multiply(A=@n22) {constB=1.5})
  n26 = Constant3Vector() {constant=<rgba 15 15 35 1>}
  Multiply(A=Add(A=Multiply(A=@n20, B=@n21), B=Multiply(A=@n25, B=@n26)), B=ScalarParameter() {defaultValue=1, parameterName=WPO_Mult})
  ScalarParameter() {parameterName=Metallic}
  ScalarParameter() {defaultValue=0.8, parameterName=Specular}
  ScalarParameter() {defaultValue=0.8, parameterName=Roughness}
 ]
}
